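# Sound cue for the Queen of Thorns boss voice lines played, by its name, when the boss loses track of its target.
# A single Random node picks one of eight equally weighted wave variations.
# Playback uses the shared henchman NPC voice attenuation, the critical enemy VO sound class and a boss combat-chatter concurrency set, with slightly reduced volume.
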SoundCue /FNM_AI_AxisThrow/Audio/VO/Boss/VO_TangyRadish_QueenOfThorns_AlertLostTarget_Cue
root=Random_0
nodes [
 {
  id=Random_0
  children [WavePlayer_0,WavePlayer_1,WavePlayer_2,WavePlayer_3,WavePlayer_4,WavePlayer_5,WavePlayer_6,WavePlayer_7]
  properties {weights [1,1,1,1,1,1,1,1]}
 }
 {id=WavePlayer_0, properties {soundWaveAssetPtr=/FNM_AI_AxisThrow/Audio/VO/Boss/wavs/VO_TangyRadish_QueenOfThorns_AlertLostTarget_010}}
 {id=WavePlayer_1, properties {soundWaveAssetPtr=/FNM_AI_AxisThrow/Audio/VO/Boss/wavs/VO_TangyRadish_QueenOfThorns_AlertLostTarget_014}}
 {id=WavePlayer_2, properties {soundWaveAssetPtr=/FNM_AI_AxisThrow/Audio/VO/Boss/wavs/VO_TangyRadish_QueenOfThorns_AlertLostTarget_020}}
 {id=WavePlayer_3, properties {soundWaveAssetPtr=/FNM_AI_AxisThrow/Audio/VO/Boss/wavs/VO_TangyRadish_QueenOfThorns_AlertLostTarget_030}}
 {id=WavePlayer_4, properties {soundWaveAssetPtr=/FNM_AI_AxisThrow/Audio/VO/Boss/wavs/VO_TangyRadish_QueenOfThorns_AlertLostTarget_041}}
 {id=WavePlayer_5, properties {soundWaveAssetPtr=/FNM_AI_AxisThrow/Audio/VO/Boss/wavs/VO_TangyRadish_QueenOfThorns_AlertLostTarget_050}}
 {id=WavePlayer_6, properties {soundWaveAssetPtr=/FNM_AI_AxisThrow/Audio/VO/Boss/wavs/VO_TangyRadish_QueenOfThorns_AlertLostTarget_060}}
 {id=WavePlayer_7, properties {soundWaveAssetPtr=/FNM_AI_AxisThrow/Audio/VO/Boss/wavs/VO_TangyRadish_QueenOfThorns_AlertLostTarget_070}}
]
properties {
 attenuationSettings=/Game/Sounds/Fort_Human_VO_BR/NPCs/Shared/NPC_VO_Henchmen_Att
 maxDistance=8500
 soundClassObject=&Fort_VO_Enemy_Critical
 volumeMultiplier=0.9
 concurrencySet [&NPC_Boss_Combat_Chatter_Conc]
}
