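# Sound cue for Queen of Thorns boss voice lines in the PlayerDefeat set, by its name played when the boss defeats a player.
# A single Random node picks among nine equally weighted wave variations.
# Output uses shared henchman NPC voice attenuation, the critical enemy VO sound class, and a boss combat chatter concurrency set to govern overlapping lines.
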SoundCue /FNM_AI_AxisThrow/Audio/VO/Boss/VO_TangyRadish_QueenOfThorns_PlayerDefeat_Cue
root=Random_0
nodes [
 {
  id=Random_0
  children [WavePlayer_0,WavePlayer_1,WavePlayer_2,WavePlayer_3,WavePlayer_4,WavePlayer_5,WavePlayer_6,WavePlayer_7,WavePlayer_8]
  properties {weights [1,1,1,1,1,1,1,1,1]}
 }
 {id=WavePlayer_0, properties {soundWaveAssetPtr=/FNM_AI_AxisThrow/Audio/VO/Boss/wavs/VO_TangyRadish_QueenOfThorns_PlayerDefeat_010}}
 {id=WavePlayer_1, properties {soundWaveAssetPtr=/FNM_AI_AxisThrow/Audio/VO/Boss/wavs/VO_TangyRadish_QueenOfThorns_PlayerDefeat_012}}
 {id=WavePlayer_2, properties {soundWaveAssetPtr=/FNM_AI_AxisThrow/Audio/VO/Boss/wavs/VO_TangyRadish_QueenOfThorns_PlayerDefeat_021}}
 {id=WavePlayer_3, properties {soundWaveAssetPtr=/FNM_AI_AxisThrow/Audio/VO/Boss/wavs/VO_TangyRadish_QueenOfThorns_PlayerDefeat_023}}
 {id=WavePlayer_4, properties {soundWaveAssetPtr=/FNM_AI_AxisThrow/Audio/VO/Boss/wavs/VO_TangyRadish_QueenOfThorns_PlayerDefeat_025}}
 {id=WavePlayer_5, properties {soundWaveAssetPtr=/FNM_AI_AxisThrow/Audio/VO/Boss/wavs/VO_TangyRadish_QueenOfThorns_PlayerDefeat_030}}
 {id=WavePlayer_6, properties {soundWaveAssetPtr=/FNM_AI_AxisThrow/Audio/VO/Boss/wavs/VO_TangyRadish_QueenOfThorns_PlayerDefeat_031}}
 {id=WavePlayer_7, properties {soundWaveAssetPtr=/FNM_AI_AxisThrow/Audio/VO/Boss/wavs/VO_TangyRadish_QueenOfThorns_PlayerDefeat_032}}
 {id=WavePlayer_8, properties {soundWaveAssetPtr=/FNM_AI_AxisThrow/Audio/VO/Boss/wavs/VO_TangyRadish_QueenOfThorns_PlayerDefeat_040}}
]
properties {
 attenuationSettings=/Game/Sounds/Fort_Human_VO_BR/NPCs/Shared/NPC_VO_Henchmen_Att
 maxDistance=8500
 soundClassObject=&Fort_VO_Enemy_Critical
 volumeMultiplier=0.9
 concurrencySet [&NPC_Boss_Combat_Chatter_Conc]
}
