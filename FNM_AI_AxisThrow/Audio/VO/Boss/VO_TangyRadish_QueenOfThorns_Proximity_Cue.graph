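# Sound cue for proximity voice lines of the TangyRadish Queen of Thorns boss in the AxisThrow AI content.
# A Random node picks one of four equally weighted atmospheric VO wave players.
# Playback uses shared henchman NPC attenuation, the critical enemy VO sound class, a slightly lowered volume and a boss combat chatter concurrency set that governs overlapping lines.
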SoundCue /FNM_AI_AxisThrow/Audio/VO/Boss/VO_TangyRadish_QueenOfThorns_Proximity_Cue
root=Random_0
nodes [
 {id=Random_0, children [WavePlayer_0,WavePlayer_1,WavePlayer_2,WavePlayer_3], properties {weights [1,1,1,1]}}
 {id=WavePlayer_0, properties {soundWaveAssetPtr=/FNM_AI_AxisThrow/Audio/VO/Boss/wavs/VO_TangyRadish_QueenOfThorns_Atmospheric_070}}
 {id=WavePlayer_1, properties {soundWaveAssetPtr=/FNM_AI_AxisThrow/Audio/VO/Boss/wavs/VO_TangyRadish_QueenOfThorns_Atmospheric_040}}
 {id=WavePlayer_2, properties {soundWaveAssetPtr=/FNM_AI_AxisThrow/Audio/VO/Boss/wavs/VO_TangyRadish_QueenOfThorns_Atmospheric_030}}
 {id=WavePlayer_3, properties {soundWaveAssetPtr=/FNM_AI_AxisThrow/Audio/VO/Boss/wavs/VO_TangyRadish_QueenOfThorns_Atmospheric_010}}
]
properties {
 attenuationSettings=/Game/Sounds/Fort_Human_VO_BR/NPCs/Shared/NPC_VO_Henchmen_Att
 maxDistance=8500
 soundClassObject=&Fort_VO_Enemy_Critical
 volumeMultiplier=0.9
 concurrencySet [&NPC_Boss_Combat_Chatter_Conc]
}
